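; Parent consumable building gameplay actor (CBGA) for forageables that players consume on first interact or gather into inventory on second.
; Once playlist data is ready, it sets the gather prompt (off in Creative), picks the gatherable item from its loot tier and may self-destruct if disabled.
; A server-only, randomly delayed weighted roll can swap it for another actor.
; Data rows drive toggles, chance and interact times; interactions send quest or gameplay events.
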(event Custom|RollForSwap
  (bind (_returnvalue_location _returnvalue_rotation _returnvalue_scale) (Math|Transform|BreakTransform (Transformation|GetActorTransform)))
  (bind (_swapactorspawnoffset_location _swapactorspawnoffset_rotation _swapactorspawnoffset_scale) (Math|Transform|BreakTransform (Variables|GetSwapActorSpawnOffset_Location)))
  ;[1 1 1] Roll For Swap - On the sever check if this CBGA should roll for a swap if it should then do a delay to space out the functionality and then roll from a row. If we hit spawn a new BGA, Init it and then after a delay destroy the original CBGA
  (if (Math|Boolean|ANDBoolean (Networking|IsServer) (Variables|RollForSwap|GetShouldRollForSwap) (Utilities|IsValidClass (Variables|RollforSwap|GetSwapActor)))
    (Utilities|FlowControl|Delay (Math|Random|RandomFloatinRange 0.1 2.0)) ; Do this on different frames as to not have it all happen at once.
    (if (Math|Random|RandomBoolwithWeight (ScalableFloat|GetValueatLevel (Variables|RollforSwap|GetRow_SwapChance)))
      (bind _returnvalue (Building|SpawnBuildingGameplayActor (Variables|RollforSwap|GetSwapActor) (Math|Transform|MakeTransform (+ _returnvalue_location _swapactorspawnoffset_location) (Math|Rotator|CombineRotators _returnvalue_rotation _swapactorspawnoffset_rotation) (+ _returnvalue_scale _swapactorspawnoffset_scale)) self))
      (Building|InitializeKismetSpawnedBuildingActor _returnvalue 0 0 true 0 false)
      (Actor|SetLifeSpan 0.1))))

(event Custom|CheckEnabled
  ;[1 1 1] Check Enabled - Read From a row to see if the CBGA should destroy itself.
  (if (not (Class|AthenaFunctionLibrary|RowtoBool (Variables|Default|GetRow_Enabled)))
    (Actor|SetLifeSpan (Math|Random|RandomFloatinRange 0.5 5.0))))

(event EventBeginPlay
  (CallFunction|SetDrawDistance)
  ;[1 1 1] Turn off second interact if not enabled or if the loot tier is None.
  (bind _gamestate (Class|FortAsyncActionCurrentPlaylistReady|PlaylistDataReadyAsync)
    (:OnReady
      (Variables|Interaction|SetbShowSecondInteractPrompt (Math|Boolean|ANDBoolean (Class|AthenaFunctionLibrary|RowtoBool (Variables|Gather|GetRow_GatherEnabled)) (Utilities|Name|NotEqual(Name) (Variables|Gather|GetGatheredLootTier) "None") (not (Math|Boolean|ORBoolean (Class|FortGameStateAthena|IsCreativeLTMPlaylist _gamestate) (Class|FortGameStateAthena|IsCreativeMMPlaylist _gamestate) (Class|FortGameStateAthena|IsCreativeModePlaylist _gamestate) false))))
      ;;[1 1 1] Call the Roll for Swap function in case we should swap.
      (CallFunction|RollforSwap)
      (CallFunction|CheckEnabled)
      (bind (_outloottodrop _returnvalue) (Loot|PickLootDrops (Variables|Gather|GetGatheredLootTier) 0 -1 "(GameplayTags=)"))
      (if _returnvalue
        (for _array_element _outloottodrop
          (bind _array_element_itemdefinition (Class|FortItemFunctionLibrary|BreakFortItemEntry _array_element))
          (bind _asfort_world_item_definition (Utilities|Casting|CastToFortWorldItemDefinition _array_element_itemdefinition))
          (Variables|BuildingGameplayActorConsumable|SetItemDefinition _asfort_world_item_definition))))))

(event Interaction|EventBlueprintOnInteract
  (params (InteractingPawn :type Fort%20Pawn%20Object%20Reference) (InteractionBeingAttempted :type EInteractionBeingAttempted%20Enum :default "FirstInteraction"))
  (switch Utilities|FlowControl|Switch|SwitchonEInteractionBeingAttempted InteractionBeingAttempted
    (:FirstInteraction
      (|CallParentFunction InteractingPawn "FirstInteraction"))
    (:SecondInteraction
      ;[1 1 1] On Second Interact Do the Gather
      (CallFunction|Gather (Fortnite|GetFortPlayerControllerfromActor InteractingPawn)))))

(event Custom|Gather
  (params (FortPlayerController :type Fort%20Player%20Controller%20Object%20Reference))
  (bind _itemdefinition (Variables|BuildingGameplayActorConsumable|GetItemDefinition))
  ;[1 1 1] On Gather Pick Loot drops from the sheets and then give the items to the player. Afterwards, Destroy Self
  (if (Networking|IsServer)
    (Inventory|GiveItemtoInventoryOwner FortPlayerController _itemdefinition)
    (Ability|SendGameplayEventtoActor (Game|Player|GetPlayerPawn FortPlayerController) "(TagName=\"Event.Interact.Gather\")" (Utilities|Struct|MakeGameplayEventData "(TagName=\"Event.Interact.Gather\")" self 0 _itemdefinition))
    (CallFunction|Destroy)))

(event Interaction|EventOnGatherOrInteract
  (params (InteractingPawn :type Fort%20Pawn%20Object%20Reference) (InteractionBeingAttempted :type EInteractionBeingAttempted%20Enum :default "FirstInteraction"))
  ;[1 1 1] Send gameplay and quest stat event if this foraged item was consumed from the ground
  (if (and (Variables|Interaction|GetbShowSecondInteractPrompt) (Utilities|Enum|Equal(Enum) InteractionBeingAttempted "FirstInteraction"))
    (Quest|SendComplexCustomStatEvent (Quests|GetQuestManager (Fortnite|GetFortPlayerControllerfromActor InteractingPawn) "Invalid") 0 "(GameplayTags=((TagName=\"Athena.Quests.ConsumeForaged\")))" (GameplayTags|GetOwnedGameplayTags :!class /Script/GameplayTags.GameplayTagAssetInterface) 1)
    (else
      (Ability|SendGameplayEventtoActor InteractingPawn "(TagName=\"Event.Forageable.Consumed\")")))
  (|CallParentFunction InteractingPawn InteractionBeingAttempted))

(fn BlueprintCanInteract
  (params (InteractingPawn :type Fort%20Pawn%20Object%20Reference) (InteractionBeingAttempted :type EInteractionBeingAttempted%20Enum :default "FirstInteraction") (InteractionType :type TInteractionType%20Enum :default "IT_NoInteraction"))
  (return (Utilities|Select :Index InteractionBeingAttempted :FirstInteraction (Variables|Interact|GetReplicatedCanInteract) :SecondInteraction (and (Variables|Interact|GetReplicatedCanInteract) (Class|AthenaFunctionLibrary|RowtoBool (Variables|Gather|GetRow_GatherEnabled))) :AllInteraction false)))

(fn BlueprintGetInteractionString
  (params (InteractingPawn :type Fort%20Pawn%20Object%20Reference) (InteractionBeingAttempted :type EInteractionBeingAttempted%20Enum :default "FirstInteraction"))
  (return (Utilities|Select :Index InteractionBeingAttempted :FirstInteraction (Variables|Interact|GetFirstInteractText) :SecondInteraction (Variables|Interact|GetSecondInteractText))))

(fn BlueprintGetInteractionTime
  (params (InteractingPawn :type Fort%20Pawn%20Object%20Reference) (InteractionBeingAttempted :type EInteractionBeingAttempted%20Enum :default "FirstInteraction"))
  (return (Utilities|Select :Index InteractionBeingAttempted :FirstInteraction (ScalableFloat|GetValueatLevel (Variables|Interact|GetRow_FirstInteractSpeed)) :SecondInteraction (ScalableFloat|GetValueatLevel (Variables|Interact|GetRow_SecondInteractSpeed)) :AllInteraction 0.0) true))

(fn Destroy
  (Actor|DestroyActor))

(fn IsTargetableByLinearMover
  (return true))

(fn SetDrawDistance
  (bind (_asfort_game_state_athena _bsuccess) (Utilities|Casting|CastToFortGameStateAthena (Game|GetGameState)))
  (for _array_element (Actor|GetComponentsByClass "/Script/Engine.PrimitiveComponent")
    (Utilities|IsValid _array_element
      (:Is%20Valid
        (LOD|SetMaxDrawDistance _array_element (* (select _bsuccess (Variables|MaxDrawDistance|GetMaxDrawDistanceAthena) (Variables|MaxDrawDistance|GetMaxDrawDistanceStw)) (Utilities|Reciprocal (Fortnite|GetNumericCVar "r.ViewDistanceScale"))))))))

(fn ShouldInvisiblePropsBeAllowed
  (return false))

(fn ConstructionScript)
