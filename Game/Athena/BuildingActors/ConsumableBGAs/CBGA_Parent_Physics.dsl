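; Parent physics consumable building actor that falls, settles, and falls again when the actor beneath it is destroyed.
; On sleep it runs a downward link check; the linked actor's destruction sets replicated bResumeSimulation, whose RepNotify wakes the physics component.
; BeginPlay plays spawn effects, then server-side resumes simulation or sleeps and links.
; Server hits trigger impact effects, and a delayed Impulse pushes the mesh from an origin before setting replicated collision.
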(event Custom|OnAwakeChanged
  (params (SimulatingComponent :type Primitive%20Component%20Object%20Reference) (bIsAwake :type Boolean))
  ;[1 1 1] On Awake Changed - If we fall asleep do a Linked Actor Check
  (if (not bIsAwake)
    (CallFunction|HandleBinding)
    (Variables|Gravity|SetbResumeSimulation false)))

(event EventBeginPlay
  ;[1 1 1] On Begin Play Check For Resim
  (|CallParentFunction)
  ;;[1 1 1] If we use the physics version bind awake changed
  (bind _outputpin (Utilities|ValidPhysicsComp)
    (:Is%20Valid
      (PhysicsObject|BindEventtoOnAwakeChanged _outputpin (EventDispatchers|CreateEvent :!event Custom|OnAwakeChanged)))
    (:Is%20Not%20Valid
      _))
  (CallFunction|SpawnFXSounds)
  (if (Variables|Gravity|GetResumeGravSimOnBegin)
    (if (Networking|IsServer)
      (Variables|Gravity|SetbResumeSimulation true))
    (elif (Networking|IsServer)
      (bind _outputpin_2 (Utilities|ValidPhysicsComp)
        (:Is%20Valid
          (PhysicsObject|PuttoSleep _outputpin_2)
          (CallFunction|HandleBinding))))))

(event Custom|SpawnFXSounds
  (bind _spawnfx (Variables|Gravity|GetSpawnFX))
  (bind _spawnsound (Variables|Gravity|GetSpawnSound))
  ;[1 1 1] Play Spawn FX and Sounds
  (Utilities|IsValid _spawnfx
    (:Is%20Valid
      (Effects|SpawnEmitteratLocation _spawnfx (Transformation|GetActorLocation) (Transformation|GetActorRotation)))
    (:Is%20Not%20Valid
      _))
  (Utilities|IsValid _spawnsound
    (:Is%20Valid
      (Audio|PlaySoundatLocation _spawnsound (Transformation|GetActorLocation)))))

(event Collision|EventHit
  (params (MyComp :type Primitive%20Component%20Object%20Reference) (Other :type Actor%20Object%20Reference) (OtherComp :type Primitive%20Component%20Object%20Reference) (bSelfMoved :type Boolean) (HitLocation :type Vector :default "0, 0, 0") (HitNormal :type Vector :default "0, 0, 0") (NormalImpulse :type Vector :default "0, 0, 0") (Hit :type Hit%20Result%20Structure%20%28by%20ref%29))
  ;[1 1 1] Handle Falling
  ;;[1 1 1] Send hit event from server to clients
  (if (Networking|IsServer)
    (CallFunction|PlayHitFX HitLocation (Math|Conversions|RotationFromXVector HitNormal))))

(event Custom|HandleBinding
  ;[1 1 1] Handle Falling
  ;;[1 1 1] Binds Event for when Building piece dies
  (LinkToActorComponent|PerformLinkingCheck (Variables|Default|GetFortLinkToActor) "Down" (Variables|Default|GetLinkedActor)))

(event Custom|PlayHitFX
  (params (Location :type Vector) (Rotation :type Rotator))
  (bind _hitgroundsound (Variables|Gravity|GetHitgroundSound))
  (bind _hitgroundfx (Variables|Gravity|GetHitgroundFX))
  ;[1 1 1] Handle Falling
  ;;[1 1 1] Play Hit Sounds and FX if valid
  (Utilities|IsValid _hitgroundfx
    (:Is%20Valid
      (Effects|SpawnEmitteratLocation _hitgroundfx Location Rotation))
    (:Is%20Not%20Valid
      _))
  (Utilities|IsValid _hitgroundsound
    (:Is%20Valid
      (Audio|PlaySoundatLocation _hitgroundsound (Transformation|GetActorLocation)))))

(event OnLinkedActorDestroyed(FortLinkToActor)
  (params (DamageCauser :type Actor%20Object%20Reference))
  ;[1 1 1] Handle Falling
  ;;[1 1 1] Base destroyed. Start simulating again.
  (if (not (Variables|Gravity|GetbResumeSimulation))
    (Variables|Gravity|SetbResumeSimulation true)
    (Networking|ForceNetUpdate)))

(event Custom|Impulse
  (params (DelayBeforeImpulse :type Float%20%28double-precision%29) (DelayBeforeCollision :type Float%20%28double-precision%29) (ImpulseOrigin :type Vector) (ImpulseAmount :type Vector) (SetCollisionAfterImpulse :type Boolean))
  ;[1 1 1] Impulse Actor
  (Utilities|FlowControl|Delay DelayBeforeImpulse)
  (Physics|AddImpulse (Variables|Default|GetStaticMesh) (* (Math|Vector|Normalize (- (Transformation|GetActorLocation) ImpulseOrigin)) ImpulseAmount) "None" false)
  (Utilities|FlowControl|Delay DelayBeforeCollision)
  (Variables|Gravity|SetRepCollision SetCollisionAfterImpulse))

(fn OnRep_bResumeSimulation
  ;[1 1 1] If the rep notify for resume its true then wake up the object.
  (if (Variables|Gravity|GetbResumeSimulation)
    (bind _outputpin (Utilities|ValidPhysicsComp)
      (:Is%20Valid
        (PhysicsObject|WakeUp _outputpin "Normal" :!class /Script/FortniteGame.FortPhysicsObjectComponent)))))

(fn OnRep_RepCollision
  (if (Variables|Gravity|GetRepCollision)
    (Collision|SetActorEnableCollision true)))

(fn ConstructionScript
  ;[1 1 1] Start with collision off so that it can be turned on later
  (if (not (Variables|Gravity|GetSpawnCollision))
    (Collision|SetActorEnableCollision false))
  (|CallParentFunction))
